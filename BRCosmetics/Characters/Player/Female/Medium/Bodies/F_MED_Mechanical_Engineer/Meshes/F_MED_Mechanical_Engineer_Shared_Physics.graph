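# Shared physics asset for the female medium Mechanical Engineer body, set up for secondary motion on dangling hip gear and the right pant cuff.
# Capsules on pelvis, spine, thighs and calves form a coarse lower-body shape joined by angularly free constraints.
# Tiny base capsules link to the pelvis with free rotation, anchoring angle-limited chains for side pouches, straps and two rear pouches.
# Four pant-bottom bodies each constrain directly to the right calf.
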PhysicsAsset /BRCosmetics/Characters/Player/Female/Medium/Bodies/F_MED_Mechanical_Engineer/Meshes/F_MED_Mechanical_Engineer_Shared_Physics
bodies {
 calf_l [{shapeType=Capsule, length=34.75, radius=8, center=<-22.5133 -1.8688 0.0003>, rotation=<rot 90 0 -0.8683>}]
 calf_r [{shapeType=Capsule, length=34.75, radius=8, center=<22.5133 1.8688 -0.0003>, rotation=<rot 90 0 179.1317>}]
 calf_twist_01_r [{shapeType=Capsule, length=9.25, radius=4.25, center=<0.0732 0.8365 -0.0001>, rotation=<rot 90 0 5.0002>}]
 dyn_hip_bk_pouch_A_a [{shapeType=Box, extentX=6.5, extentY=4.5, extentZ=16, center=<4.2769 0 0>}]
 dyn_hip_bk_pouch_A_a_base [{shapeType=Capsule, length=0.505, radius=0.505, rotation=<rot 0 -90 0>}]
 dyn_hip_bk_pouch_A_b [{shapeType=Box, extentX=7, extentY=5, extentZ=16, center=<3.8781 -0.0001 0>}]
 dyn_hip_bk_pouch_B_a [{shapeType=Box, extentX=8.5, extentY=4, extentZ=16, center=<4.5834 0 0>}]
 dyn_hip_bk_pouch_B_a_base [{shapeType=Capsule, length=0.505, radius=0.505, rotation=<rot 0 -90 0>}]
 dyn_hip_bk_pouch_B_b [{shapeType=Box, extentX=7, extentY=5, extentZ=16.5, center=<3.3579 -0.0001 0>}]
 dyn_hip_pouch_base_l [{shapeType=Capsule, length=0.505, radius=0.505, rotation=<rot 0 -90 0>}]
 dyn_hip_pouch_base_r [{shapeType=Capsule, length=0.505, radius=0.505, rotation=<rot 0 -90 0>}]
 dyn_hip_pouch_l [{shapeType=Capsule, length=10, radius=6, shapeName=dyn_hip_pouch_l_capsule, center=<8 0 5.6352>, rotation=<rot 90 -175.2364 -85.2365>}]
 dyn_hip_pouch_r [{shapeType=Box, extentX=14.5, extentY=13.5, extentZ=5, center=<-7.9793 -0.1916 -2.3652>, rotation=<rot 6.4427 0 7.5047>}]
 dyn_hip_pouch_strapA_a_l [{
   shapeType=Capsule
   length=5
   radius=2
   shapeName=dyn_hip_pouch_strapA_a_l_capsule
   center=<3.4296 0 0.6079>
   rotation=<rot 86.8076 -0.1345 -0.2724>
  }]
 dyn_hip_pouch_strapA_b_l [{
   shapeType=Capsule
   length=3.75
   radius=2.25
   shapeName=dyn_hip_pouch_strapA_b_l_capsule
   center=<3.1536 0 0.6412>
   rotation=<rot 86.2087 -0.0001 -0.0002>
  }]
 dyn_hip_pouch_strapB_a_l [{shapeType=Capsule, length=4, radius=2, shapeName=dyn_hip_pouch_strapB_a_l_capsule, center=<4.7933 0 0.9557>, rotation=<rot 90 0 0>}]
 dyn_hip_pouch_strapB_b_l [{shapeType=Capsule, length=7, radius=2, shapeName=dyn_hip_pouch_strapB_b_l_capsule, center=<6.2887 0 0.6>, rotation=<rot 90 0 0>}]
 dyn_hip_pouch_strap_a_r [{shapeType=Box, extentX=7, extentY=18.5, extentZ=1.5, center=<-3.8573 0 0>}]
 dyn_hip_pouch_strap_b_r [{shapeType=Capsule, length=9, radius=3.5, shapeName=dyn_hip_pouch_strap_b_r_capsule, center=<-5 0 -2.9082>, rotation=<rot 0 0.0002 90>}]
 dyn_pants_bottom_a_r [{shapeType=Box, extentX=12, extentY=2.5, extentZ=11, center=<6 0 0>}]
 dyn_pants_bottom_b_r [{shapeType=Box, extentX=12, extentY=11.5, extentZ=3.5, center=<6 0 0>}]
 dyn_pants_bottom_c_r [
  {
   shapeType=Capsule
   length=9.4038
   radius=1.4749
   shapeName=dyn_pants_bottom_c_r_capsule
   center=<10.1589 -0.1252 -0.3354>
   rotation=<rot 2.0441 -92.151 -0.18>
  }
  {
   shapeType=Capsule
   length=9.4038
   radius=1.4749
   shapeName=dyn_pants_bottom_c_r_capsule
   center=<2.0793 0.1775 -0.3608>
   rotation=<rot 2.0441 -92.151 -0.18>
  }
 ]
 dyn_pants_bottom_d_r [{shapeType=Box, extentX=12, extentY=11.5, extentZ=3.5, center=<6 0 0>}]
 pelvis [{shapeType=Capsule, length=15, radius=10.25, center=<-5.6244 -1.1375 -1.9324>, rotation=<rot 90 0 -1.3701>}]
 spine_02 [{shapeType=Capsule, length=4.1763, radius=7.4131, center=<0 1.7424 0>}]
 spine_03 [{shapeType=Capsule, length=4.274, radius=8.0737, center=<0 2.1666 0>}]
 thigh_l [{shapeType=Capsule, length=32.597, radius=8.7262, center=<-21.022 -0.8854 0.8473>, rotation=<rot 0.0495 -89.983 -91.9673>}]
 thigh_r [{shapeType=Capsule, length=32.597, radius=8.7262, center=<21.022 0.8854 -0.8473>, rotation=<rot 0.0495 -89.983 88.0327>}]
}
constraints [
 {
  bone1Name=thigh_l
  bone2Name=pelvis
  swing1LimitDegrees=45
  swing1Motion=Free
  swing2LimitDegrees=45
  swing2Motion=Free
  twistLimitDegrees=45
  twistMotion=Free
 }
 {
  bone1Name=spine_02
  bone2Name=pelvis
  swing1LimitDegrees=45
  swing1Motion=Free
  swing2LimitDegrees=45
  swing2Motion=Free
  twistLimitDegrees=45
  twistMotion=Free
 }
 {
  bone1Name=spine_03
  bone2Name=spine_02
  swing1LimitDegrees=45
  swing1Motion=Free
  swing2LimitDegrees=45
  swing2Motion=Free
  twistLimitDegrees=45
  twistMotion=Free
 }
 {
  bone1Name=calf_l
  bone2Name=thigh_l
  swing1LimitDegrees=45
  swing1Motion=Free
  swing2LimitDegrees=45
  swing2Motion=Free
  twistLimitDegrees=45
  twistMotion=Free
 }
 {
  bone1Name=calf_r
  bone2Name=thigh_r
  swing1LimitDegrees=45
  swing1Motion=Free
  swing2LimitDegrees=45
  swing2Motion=Free
  twistLimitDegrees=45
  twistMotion=Free
 }
 {
  bone1Name=dyn_hip_pouch_base_l
  bone2Name=pelvis
  swing1LimitDegrees=45
  swing1Motion=Free
  swing2LimitDegrees=45
  swing2Motion=Free
  twistLimitDegrees=45
  twistMotion=Free
 }
 {bone1Name=dyn_hip_pouch_l, bone2Name=dyn_hip_pouch_base_l, swing1LimitDegrees=5, swing2LimitDegrees=60, twistLimitDegrees=5}
 {bone1Name=dyn_hip_pouch_strapA_a_l, bone2Name=dyn_hip_pouch_l, swing1LimitDegrees=30, swing2LimitDegrees=60, twistLimitDegrees=20}
 {bone1Name=dyn_hip_pouch_strapA_b_l, bone2Name=dyn_hip_pouch_strapA_a_l, swing1LimitDegrees=30, swing2LimitDegrees=60, twistLimitDegrees=20}
 {bone1Name=dyn_hip_pouch_strapB_a_l, bone2Name=dyn_hip_pouch_l, swing1LimitDegrees=40, swing2LimitDegrees=60, twistLimitDegrees=20}
 {bone1Name=dyn_hip_pouch_strapB_b_l, bone2Name=dyn_hip_pouch_strapB_a_l, swing1LimitDegrees=40, swing2LimitDegrees=60, twistLimitDegrees=20}
 {
  bone1Name=dyn_hip_bk_pouch_A_a_base
  bone2Name=pelvis
  swing1LimitDegrees=45
  swing1Motion=Free
  swing2LimitDegrees=45
  swing2Motion=Free
  twistLimitDegrees=45
  twistMotion=Free
 }
 {
  bone1Name=dyn_hip_bk_pouch_B_a_base
  bone2Name=pelvis
  swing1LimitDegrees=45
  swing1Motion=Free
  swing2LimitDegrees=45
  swing2Motion=Free
  twistLimitDegrees=45
  twistMotion=Free
 }
 {bone1Name=dyn_hip_bk_pouch_A_a, bone2Name=dyn_hip_bk_pouch_A_a_base, swing1LimitDegrees=55, swing2LimitDegrees=10, twistLimitDegrees=10}
 {bone1Name=dyn_hip_bk_pouch_A_b, bone2Name=dyn_hip_bk_pouch_A_a, swing1LimitDegrees=55, swing2LimitDegrees=10, twistLimitDegrees=10}
 {bone1Name=dyn_hip_bk_pouch_B_a, bone2Name=dyn_hip_bk_pouch_B_a_base, swing1LimitDegrees=55, swing2LimitDegrees=10, twistLimitDegrees=10}
 {bone1Name=dyn_hip_bk_pouch_B_b, bone2Name=dyn_hip_bk_pouch_B_a, swing1LimitDegrees=55, swing2LimitDegrees=10, twistLimitDegrees=10}
 {
  bone1Name=dyn_hip_pouch_base_r
  bone2Name=pelvis
  swing1LimitDegrees=45
  swing1Motion=Free
  swing2LimitDegrees=45
  swing2Motion=Free
  twistLimitDegrees=45
  twistMotion=Free
 }
 {bone1Name=dyn_hip_pouch_r, bone2Name=dyn_hip_pouch_base_r, swing1LimitDegrees=10, swing2LimitDegrees=10, twistLimitDegrees=10}
 {bone1Name=dyn_hip_pouch_strap_a_r, bone2Name=dyn_hip_pouch_r, swing1LimitDegrees=15, swing2LimitDegrees=45, twistLimitDegrees=5}
 {bone1Name=dyn_hip_pouch_strap_b_r, bone2Name=dyn_hip_pouch_strap_a_r, swing1LimitDegrees=15, swing2LimitDegrees=45, twistLimitDegrees=5}
 {
  bone1Name=thigh_r
  bone2Name=pelvis
  swing1LimitDegrees=45
  swing1Motion=Free
  swing2LimitDegrees=45
  swing2Motion=Free
  twistLimitDegrees=45
  twistMotion=Free
 }
 {
  bone1Name=calf_twist_01_r
  bone2Name=calf_r
  swing1LimitDegrees=45
  swing1Motion=Free
  swing2LimitDegrees=45
  swing2Motion=Free
  twistLimitDegrees=45
  twistMotion=Free
 }
 {bone1Name=dyn_pants_bottom_a_r, bone2Name=calf_r, swing1LimitDegrees=15, swing2LimitDegrees=15, twistLimitDegrees=5}
 {bone1Name=dyn_pants_bottom_b_r, bone2Name=calf_r, swing1LimitDegrees=45, swing2LimitDegrees=45, twistLimitDegrees=45}
 {bone1Name=dyn_pants_bottom_d_r, bone2Name=calf_r, swing1LimitDegrees=45, swing2LimitDegrees=45, twistLimitDegrees=45}
 {bone1Name=dyn_pants_bottom_c_r, bone2Name=calf_r, swing1LimitDegrees=45, swing2LimitDegrees=45, twistLimitDegrees=45}
]
