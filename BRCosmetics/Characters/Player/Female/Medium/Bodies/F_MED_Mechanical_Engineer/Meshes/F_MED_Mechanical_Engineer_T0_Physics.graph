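# Physics asset for the female medium Mechanical Engineer T0 cosmetic body, covering its dangling front flaps and knot.
# Left and right two-segment box flap chains hang from small base capsules on the pelvis, with wider swing on one axis than in the other swing or twist.
# A knot on spine_02 branches into two capsule pieces with tight angular limits.
# Pelvis, spine and thigh capsules give body volume, joined by free-rotating constraints.
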PhysicsAsset /BRCosmetics/Characters/Player/Female/Medium/Bodies/F_MED_Mechanical_Engineer/Meshes/F_MED_Mechanical_Engineer_T0_Physics
bodies {
 dyn_front_flap_a_base_l [{shapeType=Capsule, length=1, radius=1}]
 dyn_front_flap_a_base_r [{shapeType=Capsule, length=1, radius=1}]
 dyn_front_flap_a_l [{shapeType=Box, extentX=8, extentY=1.5, extentZ=11.5, center=<3.9625 0.212 0>, rotation=<rot 0 1.3739 0>}]
 dyn_front_flap_a_r [{shapeType=Box, extentX=8, extentY=1.5, extentZ=11.5, center=<-3.9586 -0.3733 0>, rotation=<rot 0 1.3739 -180>}]
 dyn_front_flap_b_l [{shapeType=Box, extentX=10, extentY=2.5, extentZ=12.5, center=<4.1976 0 0>}]
 dyn_front_flap_b_r [{shapeType=Box, extentX=10, extentY=2.5, extentZ=12.5, center=<-4.1976 0 0>, rotation=<rot 0 0 180>}]
 dyn_knot [{shapeType=Capsule, length=1, radius=1}]
 dyn_knot_a [{shapeType=Capsule, length=4.75, radius=2, center=<2.2585 0 0>, rotation=<rot 90 0 0>}]
 dyn_knot_b [{shapeType=Capsule, length=4.75, radius=2, center=<1.341 0 0>, rotation=<rot 90 0 0>}]
 pelvis [{shapeType=Capsule, length=10, radius=10, center=<0 -1.4233 0>}]
 spine_02 [{shapeType=Capsule, length=4.1763, radius=7.4131, center=<0 1.7424 0>}]
 spine_03 [{shapeType=Capsule, length=4.274, radius=8.0737, center=<0 2.1666 0>}]
 thigh_l [{shapeType=Capsule, length=32.597, radius=8.7262, center=<-21.0221 -0.4158 0.8469>, rotation=<rot 0.0495 -89.983 -91.9673>}]
 thigh_r [{shapeType=Capsule, length=32.597, radius=8.7262, center=<21.022 0.8854 -0.8473>, rotation=<rot 0.0495 -89.983 88.0327>}]
}
constraints [
 {
  bone1Name=thigh_l
  bone2Name=pelvis
  swing1LimitDegrees=45
  swing1Motion=Free
  swing2LimitDegrees=45
  swing2Motion=Free
  twistLimitDegrees=45
  twistMotion=Free
 }
 {
  bone1Name=spine_02
  bone2Name=pelvis
  swing1LimitDegrees=45
  swing1Motion=Free
  swing2LimitDegrees=45
  swing2Motion=Free
  twistLimitDegrees=45
  twistMotion=Free
 }
 {
  bone1Name=spine_03
  bone2Name=spine_02
  swing1LimitDegrees=45
  swing1Motion=Free
  swing2LimitDegrees=45
  swing2Motion=Free
  twistLimitDegrees=45
  twistMotion=Free
 }
 {
  bone1Name=dyn_front_flap_a_base_l
  bone2Name=pelvis
  swing1LimitDegrees=45
  swing1Motion=Free
  swing2LimitDegrees=45
  swing2Motion=Free
  twistLimitDegrees=45
  twistMotion=Free
 }
 {
  bone1Name=dyn_front_flap_a_base_r
  bone2Name=pelvis
  swing1LimitDegrees=45
  swing1Motion=Free
  swing2LimitDegrees=45
  swing2Motion=Free
  twistLimitDegrees=45
  twistMotion=Free
 }
 {bone1Name=dyn_front_flap_a_l, bone2Name=dyn_front_flap_a_base_l, swing1LimitDegrees=45, swing2LimitDegrees=10, twistLimitDegrees=10}
 {bone1Name=dyn_front_flap_b_l, bone2Name=dyn_front_flap_a_l, swing1LimitDegrees=45, swing2LimitDegrees=10, twistLimitDegrees=10}
 {bone1Name=dyn_front_flap_a_r, bone2Name=dyn_front_flap_a_base_r, swing1LimitDegrees=45, swing2LimitDegrees=10, twistLimitDegrees=10}
 {bone1Name=dyn_front_flap_b_r, bone2Name=dyn_front_flap_a_r, swing1LimitDegrees=45, swing2LimitDegrees=10, twistLimitDegrees=10}
 {
  bone1Name=thigh_r
  bone2Name=pelvis
  swing1LimitDegrees=45
  swing1Motion=Free
  swing2LimitDegrees=45
  swing2Motion=Free
  twistLimitDegrees=45
  twistMotion=Free
 }
 {bone1Name=dyn_knot, bone2Name=spine_02, swing1LimitDegrees=10, swing2LimitDegrees=10, twistLimitDegrees=5}
 {bone1Name=dyn_knot_a, bone2Name=dyn_knot, swing1LimitDegrees=30, swing2LimitDegrees=30, twistLimitDegrees=10}
 {bone1Name=dyn_knot_b, bone2Name=dyn_knot, swing1LimitDegrees=30, swing2LimitDegrees=30, twistLimitDegrees=10}
]
