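# Physics asset for the F_MED_Mechanical_Engineer T3 female body mesh, set up for secondary motion of the front and back overall pieces.
# Each side is a chain of two boxes hung from a tiny base capsule attached to spine_02, with angular limits that keep twist tight.
# Capsules on pelvis, spine_01 and spine_02 form the torso volume, and the joints linking them and the chain bases leave rotation free.
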PhysicsAsset /BRCosmetics/Characters/Player/Female/Medium/Bodies/F_MED_Mechanical_Engineer/Meshes/F_MED_Mechanical_Engineer_T3_Physics
bodies {
 dyn_overall_bk_a [{shapeType=Box, extentX=4.5, extentY=2, extentZ=8.5, center=<2.66 0 0>}]
 dyn_overall_bk_a_base [{shapeType=Capsule, length=0.505, radius=0.505, rotation=<rot 0 -90 0>}]
 dyn_overall_bk_b [{shapeType=Box, extentX=5.5, extentY=1.5, extentZ=5.5, center=<2.9379 0 0>}]
 dyn_overall_fr_a [{shapeType=Box, extentX=5, extentY=1.5, extentZ=8, center=<2.9668 0 -0.0001>}]
 dyn_overall_fr_a_base [{shapeType=Capsule, length=0.505, radius=0.505, rotation=<rot 0 -90 0>}]
 dyn_overall_fr_b [{shapeType=Box, extentX=7, extentY=3, extentZ=5.5, center=<3.3479 0 0>}]
 pelvis [{shapeType=Capsule, length=9.9255, radius=11.3747, center=<0 -2.1129 0>}]
 spine_01 [{shapeType=Capsule, length=7.005, radius=9.255, center=<0 -0.224 0>, rotation=<rot 0 -90 0>}]
 spine_02 [{shapeType=Capsule, length=5, radius=9, center=<0 1.6634 0>}]
}
constraints [
 {bone1Name=dyn_overall_fr_a, bone2Name=dyn_overall_fr_a_base, swing1LimitDegrees=45, swing2LimitDegrees=15, twistLimitDegrees=5}
 {bone1Name=dyn_overall_fr_b, bone2Name=dyn_overall_fr_a, swing1LimitDegrees=45, swing2LimitDegrees=15, twistLimitDegrees=5}
 {bone1Name=dyn_overall_bk_a, bone2Name=dyn_overall_bk_a_base, swing1LimitDegrees=45, swing2LimitDegrees=15, twistLimitDegrees=5}
 {bone1Name=dyn_overall_bk_b, bone2Name=dyn_overall_bk_a, swing1LimitDegrees=45, swing2LimitDegrees=15, twistLimitDegrees=5}
 {
  bone1Name=dyn_overall_fr_a_base
  bone2Name=spine_02
  swing1LimitDegrees=45
  swing1Motion=Free
  swing2LimitDegrees=45
  swing2Motion=Free
  twistLimitDegrees=45
  twistMotion=Free
 }
 {
  bone1Name=dyn_overall_bk_a_base
  bone2Name=spine_02
  swing1LimitDegrees=45
  swing1Motion=Free
  swing2LimitDegrees=45
  swing2Motion=Free
  twistLimitDegrees=45
  twistMotion=Free
 }
 {
  bone1Name=spine_01
  bone2Name=pelvis
  swing1LimitDegrees=45
  swing1Motion=Free
  swing2LimitDegrees=45
  swing2Motion=Free
  twistLimitDegrees=45
  twistMotion=Free
 }
 {
  bone1Name=spine_02
  bone2Name=spine_01
  swing1LimitDegrees=45
  swing1Motion=Free
  swing2LimitDegrees=45
  swing2Motion=Free
  twistLimitDegrees=45
  twistMotion=Free
 }
]
