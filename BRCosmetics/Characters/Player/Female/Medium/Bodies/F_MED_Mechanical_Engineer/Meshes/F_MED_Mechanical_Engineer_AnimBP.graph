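# Secondary-motion AnimBlueprint for the female medium Mechanical Engineer body, copying its pose from the attached parent mesh.
# Weighted constraints position hip pouch, knot and right pant-cuff bones from body bones.
# Additive bone offsets, mostly scaled by range-mapped thigh and knee values, adjust front flaps, pouches and kneepads.
# Three chained pelvis-space RigidBody nodes then simulate the Shared, T0 and T3 physics assets; the EventGraph is empty.
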AnimBlueprint /BRCosmetics/Characters/Player/Female/Medium/Bodies/F_MED_Mechanical_Engineer/Meshes/F_MED_Mechanical_Engineer_AnimBP
target_skeleton=&/F_MED_Mechanical_Engineer_Skeleton
graphs [
 AnimGraph {nodes [
   n1 = "Constraint - dyn_hip_pouch_base_l (PPPP)"!(
    ComponentPose=Local To Component(LocalPose=Copy Pose From Mesh() {bUseAttachedParent=true})
    Alpha=1
    ConstraintWeights_0=0.6169
    ConstraintWeights_1:float=0.0257
    ConstraintWeights_2:float=0.3278
    ConstraintWeights_3:float=0.0297
   ) -> (Pose) {
    alphaBoolBlend {blendOption=Linear}
    boneToModify {boneName=dyn_hip_pouch_base_l}
    constraintSetup [
     {targetBone {boneName=pelvis}}
     {targetBone {boneName=thigh_volume_l}}
     {targetBone {boneName=thigh_twist_01_l}}
     {targetBone {boneName=thigh_l}}
    ]
    constraintWeights [1,1,1,1]
   }
   n2 = "Constraint - dyn_hip_pouch_base_r (PPP)"!(ComponentPose=@n1, Alpha=1, ConstraintWeights_0=0.7375, ConstraintWeights_1:float=0.0252, ConstraintWeights_2:float=0.2373) -> (Pose) {
    alphaBoolBlend {blendOption=Linear}
    boneToModify {boneName=dyn_hip_pouch_base_r}
    constraintSetup [{targetBone {boneName=pelvis}}, {targetBone {boneName=thigh_volume_r}}, {targetBone {boneName=thigh_twist_01_r}}]
    constraintWeights [1,1,1]
   }
   n3 = "Constraint - dyn_knot (PPPPPPP)"!(
    ComponentPose=@n2
    Alpha=1
    ConstraintWeights_0=0.0047
    ConstraintWeights_1:float=0.1386
    ConstraintWeights_2:float=0.2946
    ConstraintWeights_3:float=0.2773
    ConstraintWeights_4:float=0.2821
    ConstraintWeights_5:float=0.0005
    ConstraintWeights_6:float=0.0021
   ) -> (Pose) {
    alphaBoolBlend {blendOption=Linear}
    boneToModify {boneName=dyn_knot}
    constraintSetup [
     {targetBone {boneName=clavicle_l}}
     {targetBone {boneName=spine_01}}
     {targetBone {boneName=spine_02}}
     {targetBone {boneName=spine_03}}
     {targetBone {boneName=spine_04}}
     {targetBone {boneName=spine_05}}
     {targetBone {boneName=pelvis}}
    ]
    constraintWeights [1,1,1,1,1,1,1]
   }
   n4 = "Transform (Modify) Bone - Bone: dyn_front_flap_a_base_l"(ComponentPose=@n3, Alpha=Get ThighUpLeft() -> (ThighUpLeft:float), Translation="-5,0,0", Rotation="0.000,100,0.000", Scale="1,1,1") {
    rotationMode=BMM_Additive
    rotationSpace=BCS_BoneSpace
    translationMode=BMM_Additive
    translationSpace=BCS_BoneSpace
    alphaBoolBlend {blendOption=Linear}
    alphaScaleBiasClamp {bMapRange=true, inRange {max=0, min=-100}, outRange {max=0, min=1}}
    boneToModify {boneName=dyn_front_flap_a_base_l}
    rotation {yaw=100}
    translation {x=-5}
   }
   n5 = "Transform (Modify) Bone - Bone: dyn_front_flap_a_base_r"(ComponentPose=@n4, Alpha=Get ThighUpRight() -> (ThighUpRight:float), Translation="5,0,0", Rotation="0.000,100,0.000", Scale="1,1,1") {
    rotationMode=BMM_Additive
    rotationSpace=BCS_BoneSpace
    translationMode=BMM_Additive
    translationSpace=BCS_BoneSpace
    alphaBoolBlend {blendOption=Linear}
    alphaScaleBiasClamp {bMapRange=true, inRange {max=0, min=-100}, outRange {max=0, min=1}}
    boneToModify {boneName=dyn_front_flap_a_base_r}
    rotation {yaw=100}
    translation {x=5}
   }
   n6 = Get ThighUpLeft()
   n7 = "Transform (Modify) Bone - Bone: dyn_hip_pouch_base_l"(ComponentPose=@n5, Alpha=@n6, Translation="-2,1,1.5", Rotation="0.000,-25,-20", Scale="1,1,1") {
    rotationMode=BMM_Additive
    rotationSpace=BCS_BoneSpace
    translationMode=BMM_Additive
    translationSpace=BCS_BoneSpace
    alphaBoolBlend {blendOption=Linear}
    alphaScaleBiasClamp {bMapRange=true, inRange {max=0, min=-100}, outRange {max=0, min=1}}
    boneToModify {boneName=dyn_hip_pouch_base_l}
    rotation {roll=-20, yaw=-25}
    translation=<-2 1 1.5>
   }
   n8 = "Transform (Modify) Bone - Bone: dyn_hip_pouch_a_base_l"(ComponentPose=@n7, Alpha=@n6, Translation="0,0,3.5", Rotation="0,0,-10", Scale="1,1,1") {
    rotationMode=BMM_Additive
    rotationSpace=BCS_BoneSpace
    translationMode=BMM_Additive
    translationSpace=BCS_BoneSpace
    alphaBoolBlend {blendOption=Linear}
    alphaScaleBiasClamp {bMapRange=true, inRange {max=0, min=-100}, outRange {max=0, min=1}}
    boneToModify {boneName=dyn_hip_pouch_a_base_l}
    rotation {roll=-10}
    translation {z=3.5}
   }
   n9 = "Transform (Modify) Bone - Bone: dyn_hip_pouch_b_base_l"(ComponentPose=@n8, Alpha=@n6, Translation="0,0,1.5", Rotation="0,0,0", Scale="1,1,1") {
    rotationMode=BMM_Additive
    rotationSpace=BCS_BoneSpace
    translationMode=BMM_Additive
    translationSpace=BCS_BoneSpace
    alphaBoolBlend {blendOption=Linear}
    alphaScaleBiasClamp {bMapRange=true, inRange {max=0, min=-100}, outRange {max=0, min=1}}
    boneToModify {boneName=dyn_hip_pouch_b_base_l}
    translation {z=1.5}
   }
   n10 = "Transform (Modify) Bone - Bone: dyn_hip_pouch_base_r"(ComponentPose=@n9, Alpha=1, Translation="0,0,-1", Rotation="0,0,0", Scale="1,1,1") {
    rotationMode=BMM_Additive
    rotationSpace=BCS_BoneSpace
    translationMode=BMM_Additive
    translationSpace=BCS_BoneSpace
    alphaBoolBlend {blendOption=Linear}
    alphaScaleBiasClamp {inRange {max=25}}
    boneToModify {boneName=dyn_hip_pouch_base_r}
    translation {z=-1}
   }
   n11 = "Transform (Modify) Bone - Bone: dyn_hip_pouch_base_r"(ComponentPose=@n10, Alpha=Get ThighSideRight() -> (ThighSideRight:float), Translation="0,0,0", Rotation="20,0,-20", Scale="1,1,1") {
    rotationMode=BMM_Additive
    rotationSpace=BCS_BoneSpace
    translationMode=BMM_Additive
    translationSpace=BCS_BoneSpace
    alphaBoolBlend {blendOption=Linear}
    alphaScaleBiasClamp {bMapRange=true, inRange {max=25}}
    boneToModify {boneName=dyn_hip_pouch_base_r}
    rotation {pitch=20, roll=-20}
   }
   n12 = "Constraint - dyn_pants_bottom_a_r (P)"(ComponentPose=@n11, Alpha=1, ConstraintWeights_0=1) {
    alphaBoolBlend {blendOption=Linear}
    boneToModify {boneName=dyn_pants_bottom_a_r}
    constraintSetup [{targetBone {boneName=calf_twist_01_r}}]
    constraintWeights [1]
   }
   n13 = "Constraint - dyn_pants_bottom_b_r (P)"(ComponentPose=@n12, Alpha=1, ConstraintWeights_0=1) {
    alphaBoolBlend {blendOption=Linear}
    boneToModify {boneName=dyn_pants_bottom_b_r}
    constraintSetup [{targetBone {boneName=calf_twist_01_r}}]
    constraintWeights [1]
   }
   n14 = "Constraint - dyn_pants_bottom_c_r (P)"(ComponentPose=@n13, Alpha=1, ConstraintWeights_0=1) {
    alphaBoolBlend {blendOption=Linear}
    boneToModify {boneName=dyn_pants_bottom_c_r}
    constraintSetup [{targetBone {boneName=calf_twist_01_r}}]
    constraintWeights [1]
   }
   n15 = "Constraint - dyn_pants_bottom_d_r (P)"(ComponentPose=@n14, Alpha=1, ConstraintWeights_0=1) {
    alphaBoolBlend {blendOption=Linear}
    boneToModify {boneName=dyn_pants_bottom_d_r}
    constraintSetup [{targetBone {boneName=calf_twist_01_r}}]
    constraintWeights [1]
   }
   n16 = Get KneeLeft() -> (KneeLeft:float)
   n17 = "Transform (Modify) Bone - Bone: dyn_kneepad_up_l"(ComponentPose=@n15, Alpha=@n16, Translation="-3,-3,0", Rotation="0,50,0", Scale="1,1,1") {
    rotationMode=BMM_Additive
    rotationSpace=BCS_BoneSpace
    translationMode=BMM_Additive
    translationSpace=BCS_BoneSpace
    alphaBoolBlend {blendOption=Linear}
    alphaScaleBiasClamp {bMapRange=true, inRange {max=110}}
    boneToModify {boneName=dyn_kneepad_up_l}
    rotation {yaw=50}
    translation=<-3 -3>
   }
   n18 = "Transform (Modify) Bone - Bone: dyn_kneepad_down_l"(ComponentPose=@n17, Alpha=@n16, Translation="-2,0,0", Rotation="0,-25,0", Scale="1,1,1") {
    rotationMode=BMM_Additive
    rotationSpace=BCS_BoneSpace
    translationMode=BMM_Additive
    translationSpace=BCS_BoneSpace
    alphaBoolBlend {blendOption=Linear}
    alphaScaleBiasClamp {bMapRange=true, inRange {max=110}}
    boneToModify {boneName=dyn_kneepad_down_l}
    rotation {yaw=-25}
    translation {x=-2}
   }
   n19 = Get KneeRight() -> (KneeRight:float)
   n20 = "Transform (Modify) Bone - Bone: dyn_kneepad_up_r"(ComponentPose=@n18, Alpha=@n19, Translation="3,3,0", Rotation="0,50,0", Scale="1,1,1") {
    rotationMode=BMM_Additive
    rotationSpace=BCS_BoneSpace
    translationMode=BMM_Additive
    translationSpace=BCS_BoneSpace
    alphaBoolBlend {blendOption=Linear}
    alphaScaleBiasClamp {bMapRange=true, inRange {max=110}}
    boneToModify {boneName=dyn_kneepad_up_r}
    rotation {yaw=50}
    translation=<3 3>
   }
   n21 = "Transform (Modify) Bone - Bone: dyn_kneepad_down_r"(ComponentPose=@n20, Alpha=@n19, Translation="2,0,0", Rotation="0,-25,0", Scale="1,1,1") {
    rotationMode=BMM_Additive
    rotationSpace=BCS_BoneSpace
    translationMode=BMM_Additive
    translationSpace=BCS_BoneSpace
    alphaBoolBlend {blendOption=Linear}
    alphaScaleBiasClamp {bMapRange=true, inRange {max=110}}
    boneToModify {boneName=dyn_kneepad_down_r}
    rotation {yaw=-25}
    translation {x=2}
   }
   n22 = RigidBody(ComponentPose=@n21) {
    bClampLinearTranslationLimitToRefPose=true
    bOverrideWorldGravity=true
    overridePhysicsAsset=&/F_MED_Mechanical_Engineer_Shared_Physics
    simulationSpace=BaseBoneSpace
    alphaBoolBlend {blendOption=Linear}
    baseBoneRef {boneName=pelvis}
    componentLinearAccScale=<0.4 0.4 0.4>
    componentLinearVelScale=<0.4 0.4 0.4>
    overrideWorldGravity {z=-980}
   }
   n23 = RigidBody(ComponentPose=@n22) {
    bClampLinearTranslationLimitToRefPose=true
    bOverrideWorldGravity=true
    overridePhysicsAsset=&/F_MED_Mechanical_Engineer_T0_Physics
    simulationSpace=BaseBoneSpace
    alphaBoolBlend {blendOption=Linear}
    baseBoneRef {boneName=pelvis}
    componentLinearAccScale=<0.4 0.4 0.4>
    componentLinearVelScale=<0.4 0.4 0.4>
    overrideWorldGravity {z=-980}
   }
   n24 = RigidBody(ComponentPose=@n23) {
    bClampLinearTranslationLimitToRefPose=true
    bOverrideWorldGravity=true
    overridePhysicsAsset=&/F_MED_Mechanical_Engineer_T3_Physics
    simulationSpace=BaseBoneSpace
    alphaBoolBlend {blendOption=Linear}
    baseBoneRef {boneName=pelvis}
    componentLinearAccScale=<0.4 0.4 0.4>
    componentLinearVelScale=<0.4 0.4 0.4>
    overrideWorldGravity {z=-980}
   }
   Output Pose(Result=Component To Local(ComponentPose=@n24)) {layerGroup=None}
  ]}
 EventGraph {links [], nodes []}
]
variables []
