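# Physics asset for the shared FaceAcc part of the female medium Mechanical Engineer character, set up for two dynamic hair tails.
# Capsules on the upper spine, neck, head, clavicles and upper arms form a body proxy linked by free-swinging constraints.
# Mirrored six-segment tail chains hang from small base bodies on the head, with tight twist limits and swing limits that widen after the first two segments.
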PhysicsAsset /BRCosmetics/Characters/Player/Female/Medium/Bodies/F_MED_Mechanical_Engineer/Meshes/Parts/F_MED_Mechanical_Engineer_FaceAcc_Shared_Physics
bodies {
 clavicle_l [{shapeType=Capsule, length=7.6745, radius=7.0755, center=<4.6251 -0.9014 -0.8369>, rotation=<rot 83.5144 0.0344 -0.1132>}]
 clavicle_r [{shapeType=Capsule, length=7.1745, radius=7.5755, center=<-4.6258 1.1584 0.8368>, rotation=<rot 83.513 0.0344 179.8868>}]
 dyn_hair_tail_a_base_l [{shapeType=Capsule, length=0.505, radius=0.505, center=<0 -6 0>, rotation=<rot 0 -90 0>}]
 dyn_hair_tail_a_base_r [{shapeType=Capsule, length=0.505, radius=0.505, rotation=<rot 0 -90 0>}]
 dyn_hair_tail_a_l [{shapeType=Capsule, length=0.3159, radius=7.1148, shapeName=dyn_hair_tail_a_l_capsule, center=<3.8211 6 0>, rotation=<rot 90 0 0>}]
 dyn_hair_tail_a_r [{shapeType=Capsule, length=0.8243, radius=7.1094, shapeName=dyn_hair_tail_a_r_capsule, center=<-4.2131 -6 0>, rotation=<rot 90 0 0>}]
 dyn_hair_tail_b_l [{shapeType=Capsule, length=0.3159, radius=7.1148, shapeName=dyn_hair_tail_b_l_capsule, center=<3.8211 6 0>, rotation=<rot 90 0 0>}]
 dyn_hair_tail_b_r [{shapeType=Capsule, length=0.8243, radius=7.1094, shapeName=dyn_hair_tail_b_r_capsule, center=<-4.2131 -6 0>, rotation=<rot 90 0 0>}]
 dyn_hair_tail_c_l [{shapeType=Capsule, length=0.3159, radius=7.1148, shapeName=dyn_hair_tail_c_l_capsule, center=<3.8211 6 0>, rotation=<rot 90 0 0>}]
 dyn_hair_tail_c_r [{shapeType=Capsule, length=0.8243, radius=7.1094, shapeName=dyn_hair_tail_c_r_capsule, center=<-4.2131 -6 0>, rotation=<rot 90 0 0>}]
 dyn_hair_tail_d_l [{shapeType=Capsule, length=0.3159, radius=7.1148, shapeName=dyn_hair_tail_d_l_capsule, center=<3.8211 6 0>, rotation=<rot 90 0 0>}]
 dyn_hair_tail_d_r [{shapeType=Capsule, length=0.8243, radius=7.1094, shapeName=dyn_hair_tail_d_r_capsule, center=<-4.2131 -6 0>, rotation=<rot 90 0 0>}]
 dyn_hair_tail_e_l [{shapeType=Capsule, length=1.0958, radius=7.1634, shapeName=dyn_hair_tail_e_l_capsule, center=<3.8211 6 0>, rotation=<rot 90 0 0>}]
 dyn_hair_tail_e_r [{shapeType=Capsule, length=1.0958, radius=7.1579, shapeName=dyn_hair_tail_e_r_capsule, center=<-4.2131 -6 0>, rotation=<rot 90 0 0>}]
 dyn_hair_tail_f_l [{shapeType=Capsule, length=1.0958, radius=7.8946, shapeName=dyn_hair_tail_f_l_capsule, center=<3.8211 6 0>, rotation=<rot 90 0 0>}]
 dyn_hair_tail_f_r [{shapeType=Capsule, length=1.0958, radius=7.8891, shapeName=dyn_hair_tail_f_r_capsule, center=<-4.2131 -6 0>, rotation=<rot 90 0 0>}]
 head [{shapeType=Capsule, length=8.25, radius=6, center=<3.1816 1.3754 0.0001>, rotation=<rot 90 0 0>}]
 neck_01 [{shapeType=Capsule, length=5.7208, radius=5.5292, center=<0.8316 -0.1459 -0.1493>, rotation=<rot 90 -10.6197 -13.3382>}]
 spine_04 [{shapeType=Capsule, length=45.1001, radius=13.6788}]
 spine_05 [{shapeType=Capsule, length=7.4369, radius=9.3131, center=<0.9662 -0.4781 0>}]
 upperarm_l [{shapeType=Capsule, length=18, radius=3.5, center=<11.0156 -0.6175 -0.4439>, rotation=<rot 89.5554 180 -179.042>}]
 upperarm_r [{shapeType=Capsule, length=18, radius=3.5, center=<-11.0156 0.6175 0.4439>, rotation=<rot 89.5532 180 0.958>}]
}
constraints [
 {
  bone1Name=head
  bone2Name=neck_01
  swing1LimitDegrees=45
  swing1Motion=Free
  swing2LimitDegrees=45
  swing2Motion=Free
  twistLimitDegrees=45
  twistMotion=Free
 }
 {
  bone1Name=clavicle_l
  bone2Name=spine_05
  swing1LimitDegrees=45
  swing1Motion=Free
  swing2LimitDegrees=45
  swing2Motion=Free
  twistLimitDegrees=45
  twistMotion=Free
 }
 {
  bone1Name=spine_05
  bone2Name=spine_04
  swing1LimitDegrees=45
  swing1Motion=Free
  swing2LimitDegrees=45
  swing2Motion=Free
  twistLimitDegrees=45
  twistMotion=Free
 }
 {
  bone1Name=neck_01
  bone2Name=spine_05
  swing1LimitDegrees=45
  swing1Motion=Free
  swing2LimitDegrees=45
  swing2Motion=Free
  twistLimitDegrees=45
  twistMotion=Free
 }
 {
  bone1Name=dyn_hair_tail_a_base_l
  bone2Name=head
  swing1LimitDegrees=45
  swing1Motion=Free
  swing2LimitDegrees=45
  swing2Motion=Free
  twistLimitDegrees=45
  twistMotion=Free
 }
 {
  bone1Name=dyn_hair_tail_a_base_r
  bone2Name=head
  swing1LimitDegrees=45
  swing1Motion=Free
  swing2LimitDegrees=45
  swing2Motion=Free
  twistLimitDegrees=45
  twistMotion=Free
 }
 {
  bone1Name=upperarm_l
  bone2Name=clavicle_l
  swing1LimitDegrees=45
  swing1Motion=Free
  swing2LimitDegrees=45
  swing2Motion=Free
  twistLimitDegrees=45
  twistMotion=Free
 }
 {
  bone1Name=clavicle_r
  bone2Name=spine_05
  swing1LimitDegrees=45
  swing1Motion=Free
  swing2LimitDegrees=45
  swing2Motion=Free
  twistLimitDegrees=45
  twistMotion=Free
 }
 {
  bone1Name=upperarm_r
  bone2Name=clavicle_r
  swing1LimitDegrees=45
  swing1Motion=Free
  swing2LimitDegrees=45
  swing2Motion=Free
  twistLimitDegrees=45
  twistMotion=Free
 }
 {bone1Name=dyn_hair_tail_a_l, bone2Name=dyn_hair_tail_a_base_l, swing1LimitDegrees=15, swing2LimitDegrees=15, twistLimitDegrees=5}
 {bone1Name=dyn_hair_tail_b_l, bone2Name=dyn_hair_tail_a_l, swing1LimitDegrees=15, swing2LimitDegrees=15, twistLimitDegrees=5}
 {bone1Name=dyn_hair_tail_c_l, bone2Name=dyn_hair_tail_b_l, swing1LimitDegrees=50, swing2LimitDegrees=30, twistLimitDegrees=5}
 {bone1Name=dyn_hair_tail_d_l, bone2Name=dyn_hair_tail_c_l, swing1LimitDegrees=50, swing2LimitDegrees=30, twistLimitDegrees=5}
 {bone1Name=dyn_hair_tail_e_l, bone2Name=dyn_hair_tail_d_l, swing1LimitDegrees=50, swing2LimitDegrees=30, twistLimitDegrees=5}
 {bone1Name=dyn_hair_tail_f_l, bone2Name=dyn_hair_tail_e_l, swing1LimitDegrees=50, swing2LimitDegrees=30, twistLimitDegrees=5}
 {bone1Name=dyn_hair_tail_a_r, bone2Name=dyn_hair_tail_a_base_r, swing1LimitDegrees=15, swing2LimitDegrees=15, twistLimitDegrees=5}
 {bone1Name=dyn_hair_tail_b_r, bone2Name=dyn_hair_tail_a_r, swing1LimitDegrees=15, swing2LimitDegrees=15, twistLimitDegrees=5}
 {bone1Name=dyn_hair_tail_c_r, bone2Name=dyn_hair_tail_b_r, swing1LimitDegrees=50, swing2LimitDegrees=30, twistLimitDegrees=5}
 {bone1Name=dyn_hair_tail_d_r, bone2Name=dyn_hair_tail_c_r, swing1LimitDegrees=50, swing2LimitDegrees=30, twistLimitDegrees=5}
 {bone1Name=dyn_hair_tail_e_r, bone2Name=dyn_hair_tail_d_r, swing1LimitDegrees=50, swing2LimitDegrees=30, twistLimitDegrees=5}
 {bone1Name=dyn_hair_tail_f_r, bone2Name=dyn_hair_tail_e_r, swing1LimitDegrees=50, swing2LimitDegrees=30, twistLimitDegrees=5}
]
