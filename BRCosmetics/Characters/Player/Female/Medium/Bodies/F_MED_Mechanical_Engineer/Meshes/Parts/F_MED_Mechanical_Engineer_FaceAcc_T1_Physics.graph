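# Physics asset for the Mechanical Engineer female player's FaceAcc part, giving capsule bodies to the head and dynamic hair bones.
# Left and right front tufts each chain three segments from a small base body, while an upper hair piece has one segment on its base.
# All three bases join the head with free swing and twist.
# Segment joints use tight angular limits, with tuft swing generally widening toward the tips.
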PhysicsAsset /BRCosmetics/Characters/Player/Female/Medium/Bodies/F_MED_Mechanical_Engineer/Meshes/Parts/F_MED_Mechanical_Engineer_FaceAcc_T1_Physics
bodies {
 dyn_hair_fr_tuft_a_base_l [{shapeType=Capsule, length=0.505, radius=0.505, rotation=<rot 0 -90 0>}]
 dyn_hair_fr_tuft_a_base_r [{shapeType=Capsule, length=0.505, radius=0.505, rotation=<rot 0 -90 0>}]
 dyn_hair_fr_tuft_a_l [{shapeType=Capsule, length=1.75, radius=1, center=<2 0 0>, rotation=<rot 90 0 0>}]
 dyn_hair_fr_tuft_a_r [{shapeType=Capsule, length=1.75, radius=1, center=<-2 0 0>, rotation=<rot 90 0 180>}]
 dyn_hair_fr_tuft_b_l [{shapeType=Capsule, length=1, radius=1, center=<1.4611 0 0>, rotation=<rot 90 0 0>}]
 dyn_hair_fr_tuft_b_r [{shapeType=Capsule, length=1, radius=1, center=<-1.574 -0.0001 0>, rotation=<rot 90 0 0>}]
 dyn_hair_fr_tuft_c_l [{shapeType=Capsule, length=1.75, radius=1, center=<1.8593 0 0>, rotation=<rot 90 0 0>}]
 dyn_hair_fr_tuft_c_r [{shapeType=Capsule, length=1.75, radius=1, center=<-1.8593 0 0>, rotation=<rot 90 0 180>}]
 dyn_hair_up [{shapeType=Capsule, length=1, radius=1, center=<-0.6981 0 0>, rotation=<rot 90 0 0>}]
 dyn_hair_up_base [{shapeType=Capsule, length=0.505, radius=0.505, rotation=<rot 0 -90 0>}]
 head [{shapeType=Capsule, length=3.5461, radius=6.4463, center=<4.6233 2.925 -0.3092>, rotation=<rot 0 -64.9005 -88.2549>}]
}
constraints [
 {bone1Name=dyn_hair_up, bone2Name=dyn_hair_up_base, swing1LimitDegrees=15, swing2LimitDegrees=15, twistLimitDegrees=15}
 {bone1Name=dyn_hair_fr_tuft_a_l, bone2Name=dyn_hair_fr_tuft_a_base_l, swing1LimitDegrees=5, swing2LimitDegrees=10, twistLimitDegrees=5}
 {bone1Name=dyn_hair_fr_tuft_b_l, bone2Name=dyn_hair_fr_tuft_a_l, swing1LimitDegrees=10, swing2LimitDegrees=10, twistLimitDegrees=5}
 {bone1Name=dyn_hair_fr_tuft_c_l, bone2Name=dyn_hair_fr_tuft_b_l, swing1LimitDegrees=15, swing2LimitDegrees=10, twistLimitDegrees=5}
 {
  bone1Name=dyn_hair_fr_tuft_a_base_l
  bone2Name=head
  swing1LimitDegrees=45
  swing1Motion=Free
  swing2LimitDegrees=45
  swing2Motion=Free
  twistLimitDegrees=45
  twistMotion=Free
 }
 {
  bone1Name=dyn_hair_fr_tuft_a_base_r
  bone2Name=head
  swing1LimitDegrees=45
  swing1Motion=Free
  swing2LimitDegrees=45
  swing2Motion=Free
  twistLimitDegrees=45
  twistMotion=Free
 }
 {
  bone1Name=dyn_hair_up_base
  bone2Name=head
  swing1LimitDegrees=45
  swing1Motion=Free
  swing2LimitDegrees=45
  swing2Motion=Free
  twistLimitDegrees=45
  twistMotion=Free
 }
 {bone1Name=dyn_hair_fr_tuft_a_r, bone2Name=dyn_hair_fr_tuft_a_base_r, swing1LimitDegrees=5, swing2LimitDegrees=10, twistLimitDegrees=5}
 {bone1Name=dyn_hair_fr_tuft_b_r, bone2Name=dyn_hair_fr_tuft_a_r, swing1LimitDegrees=10, swing2LimitDegrees=10, twistLimitDegrees=5}
 {bone1Name=dyn_hair_fr_tuft_c_r, bone2Name=dyn_hair_fr_tuft_b_r, swing1LimitDegrees=15, swing2LimitDegrees=15, twistLimitDegrees=5}
]
