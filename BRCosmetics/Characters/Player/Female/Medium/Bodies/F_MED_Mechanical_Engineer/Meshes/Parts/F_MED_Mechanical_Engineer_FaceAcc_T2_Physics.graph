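# Physics asset for the face accessory (T2) of the female medium Mechanical Engineer player cosmetic, covering its helmet bones.
# It has two single-capsule bodies: a small one on dyn_helmet_base and a larger, offset and rotated one on dyn_helmet.
# One constraint joins the two with tight swing and twist limits, so the helmet can move only a few degrees relative to its base.
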PhysicsAsset /BRCosmetics/Characters/Player/Female/Medium/Bodies/F_MED_Mechanical_Engineer/Meshes/Parts/F_MED_Mechanical_Engineer_FaceAcc_T2_Physics
bodies {
 dyn_helmet [{shapeType=Capsule, length=4.5, radius=2.5, center=<3.855 0 0>, rotation=<rot 90 0 0>}]
 dyn_helmet_base [{shapeType=Capsule, length=1, radius=1}]
}
constraints [{bone1Name=dyn_helmet, bone2Name=dyn_helmet_base, swing1LimitDegrees=5, swing2LimitDegrees=1, twistLimitDegrees=5}]
