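# Animation Blueprint for the female Mechanical Engineer's face accessory part, posing and simulating hair tails, nose ring and helmet.
# Copies parent and body poses, constrains the nose ring, offsets tails and helmet, then caches the result.
# Two branches add HeadUp/HeadTwist-driven tail corrections; only one adds a bone constraint and four chained RigidBody simulations before a bool blend.
# Initialization caches meshes and pawn, sets front-end alphas and computes gravity override.
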AnimBlueprint /BRCosmetics/Characters/Player/Female/Medium/Bodies/F_MED_Mechanical_Engineer/Meshes/Parts/F_MED_Mechanical_Engineer_FaceAcc_AnimBP
target_skeleton=&/F_MED_Mechanical_Engineer_FaceAcc_Skeleton
graphs [
 AnimGraph {nodes [
   n1 = Get bIsTargeting()
   n2 = "Transform (Modify) Bone - Bone: dyn_hair_tail_a_base_l"!(ComponentPose, bAlphaBoolEnabled:bool=@n1, Translation="0,0,1", Rotation="-5,0,0", Scale="1,1,1") -> (Pose) {
    alphaInputType=Bool
    rotationMode=BMM_Additive
    rotationSpace=BCS_BoneSpace
    translationMode=BMM_Additive
    translationSpace=BCS_BoneSpace
    alphaBoolBlend {blendOption=Linear}
    alphaScaleBiasClamp {inRange {max=0, min=-50}, outRange {max=0, min=1}}
    boneToModify {boneName=dyn_hair_tail_a_base_l}
    rotation {pitch=-5}
    translation {z=1}
   }
   "Transform (Modify) Bone - Bone: dyn_hair_tail_a_base_r"!(ComponentPose=@n2, bAlphaBoolEnabled:bool=@n1, Translation="0,0,0", Rotation="-15,0,0", Scale="1,1,1") -> (Pose) {
    alphaInputType=Bool
    rotationMode=BMM_Additive
    rotationSpace=BCS_BoneSpace
    translationMode=BMM_Additive
    translationSpace=BCS_BoneSpace
    alphaBoolBlend {blendInTime=0.1, blendOption=Linear, blendOutTime=0.1}
    alphaScaleBiasClamp {inRange {max=0, min=-50}, outRange {max=0, min=1}}
    boneToModify {boneName=dyn_hair_tail_a_base_r}
    rotation {pitch=-15}
   }
   RigidBody!(ComponentPose, Alpha, ExternalForce, bTransferBoneVelocities:bool=False) -> (Pose) {
    bClampLinearTranslationLimitToRefPose=true
    bOverrideWorldGravity=true
    bTransferBoneVelocities=true
    overridePhysicsAsset=&/F_MED_Mechanical_Engineer_FaceAcc_Shared_Physics
    simulationSpace=BaseBoneSpace
    alphaBoolBlend {blendOption=Linear}
    baseBoneRef {boneName=root}
    componentAppliedLinearAccClamp=<5000 5000 5000>
    componentLinearAccScale=<0.4 0.4 0.5>
    componentLinearVelScale=<0.4 0.4 0.5>
    overrideWorldGravity {z=-980}
    simSpaceSettings {
     maxAngularAcceleration=1000
     maxAngularVelocity=3
     maxLinearAcceleration=700
     maxLinearVelocity=500
     velocityScaleZ=0
     worldAlpha=1
     externalLinearDragV=<0.2 0.2>
    }
   }
   RigidBody(bTransferBoneVelocities=False) {
    bClampLinearTranslationLimitToRefPose=true
    bOverrideWorldGravity=true
    bTransferBoneVelocities=true
    overridePhysicsAsset=&F_MED_Mechanical_Engineer_FaceAcc_Shared_Physics
    simulationSpace=BaseBoneSpace
    alphaBoolBlend {blendOption=Linear}
    baseBoneRef {boneName=root}
    componentAppliedLinearAccClamp=<5000 5000 5000>
    componentLinearAccScale=<0.4 0.4 0.5>
    componentLinearVelScale=<0.4 0.4 0.5>
    overrideWorldGravity {z=-980}
    simSpaceSettings {
     maxAngularAcceleration=1000
     maxAngularVelocity=3
     maxLinearAcceleration=700
     maxLinearVelocity=500
     velocityScaleZ=0
     externalLinearDragV=<0.2 0.2>
    }
   }
   n3 = Get HeadUp() -> (HeadUp:float)
   n4 = "Transform (Modify) Bone - Bone: dyn_hair_tail_a_base_l"!(
    ComponentPose=Local To Component(LocalPose=Use cached pose 'PreRBAN'())
    Alpha:float=@n3
    Translation="0,-4,0"
    Rotation="-5,-45,0"
    Scale="1,1,1"
   ) -> (Pose) {
    rotationMode=BMM_Additive
    rotationSpace=BCS_BoneSpace
    translationMode=BMM_Additive
    translationSpace=BCS_BoneSpace
    alphaBoolBlend {blendOption=Linear}
    alphaScaleBiasClamp {bMapRange=true, inRange {max=0, min=-50}, outRange {max=0, min=1}}
    boneToModify {boneName=dyn_hair_tail_a_base_l}
    rotation {pitch=-5, yaw=-45}
    translation {y=-4}
   }
   n5 = "Transform (Modify) Bone - Bone: dyn_hair_tail_a_base_l"(ComponentPose=@n4, Alpha=@n3, Translation="0,0,0", Rotation="0,15,0", Scale="1,1,1") {
    rotationMode=BMM_Additive
    rotationSpace=BCS_BoneSpace
    translationMode=BMM_Additive
    translationSpace=BCS_BoneSpace
    alphaBoolBlend {blendOption=Linear}
    alphaScaleBiasClamp {bMapRange=true, inRange {max=15}}
    boneToModify {boneName=dyn_hair_tail_a_base_l}
    rotation {yaw=15}
   }
   n6 = Get HeadUp()
   n7 = "Transform (Modify) Bone - Bone: dyn_hair_tail_a_base_r"!(ComponentPose=@n5, Alpha:float=@n6, Translation="0,4,0", Rotation="-5,-45,0", Scale="1,1,1") -> (Pose) {
    rotationMode=BMM_Additive
    rotationSpace=BCS_BoneSpace
    translationMode=BMM_Additive
    translationSpace=BCS_BoneSpace
    alphaBoolBlend {blendOption=Linear}
    alphaScaleBiasClamp {bMapRange=true, inRange {max=0, min=-50}, outRange {max=0, min=1}}
    boneToModify {boneName=dyn_hair_tail_a_base_r}
    rotation {pitch=-5, yaw=-45}
    translation {y=4}
   }
   n8 = "Transform (Modify) Bone - Bone: dyn_hair_tail_a_base_r"(ComponentPose=@n7, Alpha=@n6, Translation="0,0,0", Rotation="0,15,0", Scale="1,1,1") {
    rotationMode=BMM_Additive
    rotationSpace=BCS_BoneSpace
    translationMode=BMM_Additive
    translationSpace=BCS_BoneSpace
    alphaBoolBlend {blendOption=Linear}
    alphaScaleBiasClamp {bMapRange=true, inRange {max=15}}
    boneToModify {boneName=dyn_hair_tail_a_base_r}
    rotation {yaw=15}
   }
   n9 = Get HeadTwist() -> (HeadTwist:float)
   n10 = "Transform (Modify) Bone - Bone: dyn_hair_tail_a_base_l"(ComponentPose=@n8, Alpha=@n9, Translation="0,5,0", Rotation="-20,0,0", Scale="1,1,1") {
    rotationMode=BMM_Additive
    rotationSpace=BCS_BoneSpace
    translationMode=BMM_Additive
    translationSpace=BCS_BoneSpace
    alphaBoolBlend {blendOption=Linear}
    alphaScaleBiasClamp {bMapRange=true, inRange {max=40}}
    boneToModify {boneName=dyn_hair_tail_a_base_l}
    rotation {pitch=-20}
    translation {y=5}
   }
   n11 = "Transform (Modify) Bone - Bone: dyn_hair_tail_a_base_l"(ComponentPose=@n10, Alpha=@n9, Translation="0,-2.5,0", Rotation="20,-15,0", Scale="1,1,1") {
    rotationMode=BMM_Additive
    rotationSpace=BCS_BoneSpace
    translationMode=BMM_Additive
    translationSpace=BCS_BoneSpace
    alphaBoolBlend {blendOption=Linear}
    alphaScaleBiasClamp {bMapRange=true, inRange {max=0, min=-40}, outRange {max=0, min=1}}
    boneToModify {boneName=dyn_hair_tail_a_base_l}
    rotation {pitch=20, yaw=-15}
    translation {y=-2.5}
   }
   n12 = Get HeadTwist()
   n13 = "Transform (Modify) Bone - Bone: dyn_hair_tail_a_base_r"(ComponentPose=@n11, Alpha=@n12, Translation="0,2.5,0", Rotation="20,-25,0", Scale="1,1,1") {
    rotationMode=BMM_Additive
    rotationSpace=BCS_BoneSpace
    translationMode=BMM_Additive
    translationSpace=BCS_BoneSpace
    alphaBoolBlend {blendOption=Linear}
    alphaScaleBiasClamp {bMapRange=true, inRange {max=40}}
    boneToModify {boneName=dyn_hair_tail_a_base_r}
    rotation {pitch=20, yaw=-25}
    translation {y=2.5}
   }
   n14 = "Transform (Modify) Bone - Bone: dyn_hair_tail_a_base_r"(ComponentPose=@n13, Alpha=@n12, Translation="0,-3,-3", Rotation="-20,0,0", Scale="1,1,1") {
    rotationMode=BMM_Additive
    rotationSpace=BCS_BoneSpace
    translationMode=BMM_Additive
    translationSpace=BCS_BoneSpace
    alphaBoolBlend {blendOption=Linear}
    alphaScaleBiasClamp {bMapRange=true, inRange {max=0, min=-40}, outRange {max=0, min=1}}
    boneToModify {boneName=dyn_hair_tail_a_base_r}
    rotation {pitch=-20}
    translation {y=-3, z=-3}
   }
   n15 = Get HeadUp()
   n16 = "Transform (Modify) Bone - Bone: dyn_hair_tail_a_base_l"(
    ComponentPose=Local To Component(LocalPose=Use cached pose 'PreRBAN'())
    Alpha=@n15
    Translation="0,-4,0"
    Rotation="-5,-45,0"
    Scale="1,1,1"
   ) {
    rotationMode=BMM_Additive
    rotationSpace=BCS_BoneSpace
    translationMode=BMM_Additive
    translationSpace=BCS_BoneSpace
    alphaBoolBlend {blendOption=Linear}
    alphaScaleBiasClamp {bMapRange=true, inRange {max=0, min=-50}, outRange {max=0, min=1}}
    boneToModify {boneName=dyn_hair_tail_a_base_l}
    rotation {pitch=-5, yaw=-45}
    translation {y=-4}
   }
   n17 = "Transform (Modify) Bone - Bone: dyn_hair_tail_a_base_l"(ComponentPose=@n16, Alpha=@n15, Translation="0,0,0", Rotation="0,15,0", Scale="1,1,1") {
    rotationMode=BMM_Additive
    rotationSpace=BCS_BoneSpace
    translationMode=BMM_Additive
    translationSpace=BCS_BoneSpace
    alphaBoolBlend {blendOption=Linear}
    alphaScaleBiasClamp {bMapRange=true, inRange {max=15}}
    boneToModify {boneName=dyn_hair_tail_a_base_l}
    rotation {yaw=15}
   }
   n18 = Get HeadUp()
   n19 = "Transform (Modify) Bone - Bone: dyn_hair_tail_a_base_r"(ComponentPose=@n17, Alpha=@n18, Translation="0,4,0", Rotation="-5,-45,0", Scale="1,1,1") {
    rotationMode=BMM_Additive
    rotationSpace=BCS_BoneSpace
    translationMode=BMM_Additive
    translationSpace=BCS_BoneSpace
    alphaBoolBlend {blendOption=Linear}
    alphaScaleBiasClamp {bMapRange=true, inRange {max=0, min=-50}, outRange {max=0, min=1}}
    boneToModify {boneName=dyn_hair_tail_a_base_r}
    rotation {pitch=-5, yaw=-45}
    translation {y=4}
   }
   n20 = "Transform (Modify) Bone - Bone: dyn_hair_tail_a_base_r"(ComponentPose=@n19, Alpha=@n18, Translation="0,0,0", Rotation="0,15,0", Scale="1,1,1") {
    rotationMode=BMM_Additive
    rotationSpace=BCS_BoneSpace
    translationMode=BMM_Additive
    translationSpace=BCS_BoneSpace
    alphaBoolBlend {blendOption=Linear}
    alphaScaleBiasClamp {bMapRange=true, inRange {max=15}}
    boneToModify {boneName=dyn_hair_tail_a_base_r}
    rotation {yaw=15}
   }
   n21 = Get HeadTwist()
   n22 = "Transform (Modify) Bone - Bone: dyn_hair_tail_a_base_l"(ComponentPose=@n20, Alpha=@n21, Translation="0,0,0", Rotation="-20,0,0", Scale="1,1,1") {
    rotationMode=BMM_Additive
    rotationSpace=BCS_BoneSpace
    translationMode=BMM_Additive
    translationSpace=BCS_BoneSpace
    alphaBoolBlend {blendOption=Linear}
    alphaScaleBiasClamp {bMapRange=true, inRange {max=40}}
    boneToModify {boneName=dyn_hair_tail_a_base_l}
    rotation {pitch=-20}
   }
   n23 = "Transform (Modify) Bone - Bone: dyn_hair_tail_a_base_l"(ComponentPose=@n22, Alpha=@n21, Translation="0,-2.5,0", Rotation="20,-30,0", Scale="1,1,1") {
    rotationMode=BMM_Additive
    rotationSpace=BCS_BoneSpace
    translationMode=BMM_Additive
    translationSpace=BCS_BoneSpace
    alphaBoolBlend {blendOption=Linear}
    alphaScaleBiasClamp {bMapRange=true, inRange {max=0, min=-40}, outRange {max=0, min=1}}
    boneToModify {boneName=dyn_hair_tail_a_base_l}
    rotation {pitch=20, yaw=-30}
    translation {y=-2.5}
   }
   n24 = Get HeadTwist()
   n25 = "Transform (Modify) Bone - Bone: dyn_hair_tail_a_base_r"(ComponentPose=@n23, Alpha=@n24, Translation="0,2.5,0", Rotation="20,-25,0", Scale="1,1,1") {
    rotationMode=BMM_Additive
    rotationSpace=BCS_BoneSpace
    translationMode=BMM_Additive
    translationSpace=BCS_BoneSpace
    alphaBoolBlend {blendOption=Linear}
    alphaScaleBiasClamp {bMapRange=true, inRange {max=40}}
    boneToModify {boneName=dyn_hair_tail_a_base_r}
    rotation {pitch=20, yaw=-25}
    translation {y=2.5}
   }
   n26 = "Transform (Modify) Bone - Bone: dyn_hair_tail_a_base_r"(ComponentPose=@n25, Alpha=@n24, Translation="0,0,0", Rotation="-20,0,0", Scale="1,1,1") {
    rotationMode=BMM_Additive
    rotationSpace=BCS_BoneSpace
    translationMode=BMM_Additive
    translationSpace=BCS_BoneSpace
    alphaBoolBlend {blendOption=Linear}
    alphaScaleBiasClamp {bMapRange=true, inRange {max=0, min=-40}, outRange {max=0, min=1}}
    boneToModify {boneName=dyn_hair_tail_a_base_r}
    rotation {pitch=-20}
   }
   n27 = "Constraint - dyn_hair_tail_b_r (P)"(ComponentPose=@n26, Alpha=0.2, ConstraintWeights_0=1) {
    alphaBoolBlend {blendOption=Linear}
    boneToModify {boneName=dyn_hair_tail_b_r}
    constraintSetup [{targetBone {boneName=spine_05}}]
    constraintWeights [1]
   }
   n28 = RigidBody!(ComponentPose=@n27, Alpha, OverrideWorldGravity:Vector="0,0,0", ExternalForce) -> (Pose) {
    bClampLinearTranslationLimitToRefPose=true
    overridePhysicsAsset=&F_MED_Mechanical_Engineer_FaceAcc_Shared_Physics
    simulationSpace=BaseBoneSpace
    alphaBoolBlend {blendOption=Linear}
    baseBoneRef {boneName=neck_01}
    componentAppliedLinearAccClamp=<5000 5000 5000>
    componentLinearAccScale=<0.4 0.4 0.5>
    componentLinearVelScale=<0.4 0.4 0.5>
    simSpaceSettings {
     maxAngularAcceleration=1000
     maxAngularVelocity=3
     maxLinearAcceleration=700
     maxLinearVelocity=500
     velocityScaleZ=0
     worldAlpha=1
     externalLinearDragV=<0.2 0.2>
    }
   }
   n29 = RigidBody!(ComponentPose=@n28, Alpha, ExternalForce, bTransferBoneVelocities:bool=False) -> (Pose) {
    bClampLinearTranslationLimitToRefPose=true
    bOverrideWorldGravity=true
    bTransferBoneVelocities=true
    overridePhysicsAsset=&F_MED_Mechanical_Engineer_FaceAcc_Shared_Physics
    simulationSpace=BaseBoneSpace
    alphaBoolBlend {blendOption=Linear}
    baseBoneRef {boneName=neck_01}
    overrideWorldGravity {z=-980}
    simSpaceSettings {maxAngularAcceleration=100, maxAngularVelocity=100, maxLinearAcceleration=100, maxLinearVelocity=100}
   }
   n30 = RigidBody!(ComponentPose=@n29, Alpha, ExternalForce) -> (Pose) {
    bClampLinearTranslationLimitToRefPose=true
    bOverrideWorldGravity=true
    overridePhysicsAsset=&/F_MED_Mechanical_Engineer_FaceAcc_T1_Physics
    simulationSpace=BaseBoneSpace
    alphaBoolBlend {blendOption=Linear}
    baseBoneRef {boneName=neck_01}
    componentAppliedLinearAccClamp=<700 700 700>
    componentLinearAccScale=<0.35 0.45 0.45>
    componentLinearVelScale=<0.3 0.3 0.3>
    overrideWorldGravity {z=-150}
   }
   n31 = RigidBody(ComponentPose=@n30, Alpha=0.3) {
    bClampLinearTranslationLimitToRefPose=true
    bOverrideWorldGravity=true
    overridePhysicsAsset=&/F_MED_Mechanical_Engineer_FaceAcc_T2_Physics
    simulationSpace=BaseBoneSpace
    alphaBoolBlend {blendOption=Linear}
    baseBoneRef {boneName=neck_01}
    componentLinearAccScale=<0.1 0.1 0.1>
    componentLinearVelScale=<0.1 0.1 0.1>
    overrideWorldGravity {z=-980}
   }
   Output Pose(Result=Blend Poses by bool(BlendPose_0=Component To Local(ComponentPose=@n31), BlendPose_1=Component To Local(ComponentPose=@n14)) {blendType=Linear}) {layerGroup=None}
   n32 = Copy Pose From Mesh(SourceMeshComponent=Get Body_Mesh())
   n33 = Copy Pose From Mesh() {bCopyCurves=true, bUseAttachedParent=true}
   n34 = "Constraint - C_nose_ring (PPPPP)"!(
    ComponentPose=Local To Component(LocalPose=Layered blend per bone(BasePose=@n33, BlendPoses_0=@n32) {layerSetup [{branchFilters [{boneName=upperarm_l}, {boneName=upperarm_r}]}]})
    Alpha=1
    ConstraintWeights_0=0
    ConstraintWeights_1:float=0
    ConstraintWeights_2:float=0.3
    ConstraintWeights_3:float=0.7
    ConstraintWeights_4:float=0
   ) -> (Pose) {
    alphaBoolBlend {blendOption=Linear}
    boneToModify {boneName=C_nose_ring}
    constraintSetup [
     {targetBone {boneName=L_lip_upper_outer}}
     {targetBone {boneName=L_cheek_inner}}
     {targetBone {boneName=C_nose_bridge}}
     {targetBone {boneName=head}}
     {targetBone {boneName=C_lip_upper_mid}}
    ]
    constraintWeights [1,1,1,1,1]
   }
   n35 = "Transform (Modify) Bone - Bone: dyn_hair_tail_a_base_l"(ComponentPose=@n34, Alpha=1, Translation="0,0,-1", Rotation="-10,0,0", Scale="1,1,1") {
    rotationMode=BMM_Additive
    rotationSpace=BCS_BoneSpace
    translationMode=BMM_Additive
    translationSpace=BCS_BoneSpace
    alphaBoolBlend {blendOption=Linear}
    alphaScaleBiasClamp {inRange {max=0, min=-50}, outRange {max=0, min=1}}
    boneToModify {boneName=dyn_hair_tail_a_base_l}
    rotation {pitch=-10}
    translation {z=-1}
   }
   n36 = "Transform (Modify) Bone - Bone: dyn_hair_tail_a_base_r"(ComponentPose=@n35, Alpha=1, Translation="0,0,0", Rotation="-10,0,0", Scale="1,1,1") {
    rotationMode=BMM_Additive
    rotationSpace=BCS_BoneSpace
    translationMode=BMM_Additive
    translationSpace=BCS_BoneSpace
    alphaBoolBlend {blendOption=Linear}
    alphaScaleBiasClamp {inRange {max=0, min=-50}, outRange {max=0, min=1}}
    boneToModify {boneName=dyn_hair_tail_a_base_r}
    rotation {pitch=-10}
   }
   n37 = "Transform (Modify) Bone - Bone: dyn_helmet_base"(ComponentPose=@n36, Alpha=Get HelmetUp() -> (HelmetUp:float), Translation="0,0,0", Rotation="0,70,0", Scale="1,1,1") {
    rotationMode=BMM_Additive
    rotationSpace=BCS_BoneSpace
    translationMode=BMM_Additive
    translationSpace=BCS_BoneSpace
    alphaBoolBlend {blendOption=ExpInOut, blendOutTime=1}
    alphaScaleBiasClamp {bInterpResult=true, interpSpeedDecreasing=20, interpSpeedIncreasing=20, inRange {max=40}}
    boneToModify {boneName=dyn_helmet_base}
    rotation {yaw=70}
   }
   Save cached pose 'PreRBAN'(Pose=Component To Local(ComponentPose=@n37)) {cacheName=PreRBAN}
  ]}
 EventGraph {nodes [
   Event BlueprintInitializeAnimation()
   > n1 = Sequence()
   > Set Body_Mesh(Body_Mesh=GetOwnerSkeletalMeshComponent())
   > n2 = Cast To PlayerPawn_Athena(Object=TryGetPawnOwner())
   > n3 = Set Pawn Owner(Pawn Owner:Player Pawn Athena Object Reference=@n2) -> (Output_Get:Player Pawn Athena Object Reference)
   > Set Head_Mesh(Head_Mesh:Skeletal Mesh Component Object Reference=GetSkeletalMeshForPartType(self=@n3, PartType=Head)) -> (Output_Get:Skeletal Mesh Component Object Reference)
   > n4 = Branch(Condition=Get bIsInFrontEnd())
   > Set frontend alpha(frontend alpha:double=0.9) -> (Output_Get:double)
   > Set normalAlpha(normalAlpha:double=0) -> (Output_Get:double)
   Set frontend alpha(execute=@n4.else, frontend alpha=0)
   > Set normalAlpha(normalAlpha=1)
   n5 = Gravity Override State Function DEPRECATED(
    execute=@n1.then_1
    CharacterPartAnimInstance=Self-Reference()
    GameTimeInSeconds=GetGameTimeInSeconds()
    GravityOverrideParams=Get Gravity Override Params()
   )
   > Set Gravity Override(Gravity Override=@n5)
  ]}
]
variables [
 Body_Mesh {type=SkeletalMeshComponent}
 frontend alpha {type=number}
 Gravity Override {type=Vector}
 Gravity Override Params {type=GravityOverrideParamsStruct}
 Head_Mesh {type=SkeletalMeshComponent}
 normalAlpha {type=number}
 Pawn Owner {type=PlayerPawn_Athena_C}
]
